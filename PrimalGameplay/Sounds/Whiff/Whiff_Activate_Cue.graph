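# Sound cue that plays the Whiff activation sound in PrimalGameplay, classed under Fort gadget and item sounds.
# A single wave player feeds the Whiff_Activate_01 wave straight to the output, with no randomization or mixing.
# Volume is reduced, and a shared Whiff attenuation asset from the base game handles spatial falloff.
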SoundCue /PrimalGameplay/Sounds/Whiff/Whiff_Activate_Cue
root=WavePlayer_2
nodes [{id=WavePlayer_2, properties {soundWaveAssetPtr=&/Whiff_Activate_01}}]
properties {
 attenuationSettings=/Game/Sounds/Attenuation/Fort_Whiff_Att
 maxDistance=3584
 soundClassObject=&Fort_Gadgets_Items
 volumeMultiplier=0.6
 attenuationOverrides {distanceAlgorithm=Logarithmic, falloffDistance=500, attenuationShapeExtents {x=0}}
}
